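# Control Rig for the TerrierCure backpack that procedurally sways its tentacle bones.
# From BeginExecution, a serial chain of ChainHarmonicsPerItem units applies sine-eased waves, with pendulum and reach disabled.
# Most tentacles layer a unit on the CH bone with another further down, each tuned separately for varied motion.
# The deform_tentacle_r_2 unit is disconnected from execution, the head and eyelid bones are untouched, and debug drawing is enabled throughout.
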ControlRigBlueprint /BRCosmetics/Accessories/FORT_Backpacks/Backpack_M_MED_TerrierCure/Meshes/M_MED_TerrierCure_Pack_CtrlRig
graphs [
 RigVMFunctionLibrary {links [], nodes []}
 RigVMModel {nodes [
   RigUnit_BeginExecution = &RigUnit_BeginExecution()
   /Script/ControlRig.RigUnit_ChainHarmonicsPerItem(
    ChainRoot=(Type=Bone,Name="deform_tentacle_b_1")
    DrawWorldOffset=(Rotation=(X=0,Y=0,Z=0,W=1),Translation=(X=0,Y=0,Z=0),Scale3D=(X=1,Y=1,Z=1))
    Pendulum=(bEnabled=False,PendulumStiffness=2,PendulumGravity=(X=0,Y=0,Z=0),PendulumBlend=0.75,PendulumDrag=0.98,PendulumMinimum=0,PendulumMaximum=0.1,PendulumEase=Linear,UnwindAxis=(X=0,Y=1,Z=0),UnwindMinimum=0.2,UnwindMaximum=0.05)
    Reach=(bEnabled=False,ReachTarget=(X=0,Y=0,Z=0),ReachAxis=(X=1,Y=0,Z=0),ReachMinimum=0,ReachMaximum=0,ReachEase=Linear)
    Speed=(X=0.6,Y=0.8,Z=1.2)
    Wave=(bEnabled=True,WaveFrequency=(X=0.2,Y=0.6,Z=0.8),WaveAmplitude=(X=0,Y=0.5,Z=0),WaveOffset=(X=0,Y=1,Z=1),WaveNoise=(X=0,Y=0,Z=0),WaveMinimum=0,WaveMaximum=1,WaveEase=SineEaseInOut)
    WaveCurve="(EditorCurveData=(Keys=((InterpMode=RCIM_Linear,TangentMode=RCTM_Auto,TangentWeightMode=RCTWM_WeightedNone,Time=0,Value=0,ArriveTangent=0,ArriveTangentWeight=0,LeaveTangent=0,LeaveTangentWeight=0),(InterpMode=RCIM_Linear,TangentMode=RCTM_Auto,TangentWeight ...(truncated)"
    bDrawDebug=True
    ExecutePin=@RigUnit_BeginExecution.ExecutePin
   ) {is=struct RigVMUnitNode}
   > &RigUnit_ChainHarmonicsPerItem(
    ChainRoot=(Type=Bone,Name="deform_tentacle_l_b_CH")
    DrawWorldOffset=(Rotation=(X=0,Y=0,Z=0,W=1),Translation=(X=0,Y=0,Z=0),Scale3D=(X=1,Y=1,Z=1))
    Pendulum=(bEnabled=False,PendulumStiffness=2,PendulumGravity=(X=0,Y=0,Z=0),PendulumBlend=0.75,PendulumDrag=0.98,PendulumMinimum=0,PendulumMaximum=0.1,PendulumEase=Linear,UnwindAxis=(X=0,Y=1,Z=0),UnwindMinimum=0.2,UnwindMaximum=0.05)
    Reach=(bEnabled=False,ReachTarget=(X=0,Y=0,Z=0),ReachAxis=(X=1,Y=0,Z=0),ReachMinimum=0,ReachMaximum=0,ReachEase=Linear)
    Speed=(X=0.8,Y=1.3,Z=1.5)
    Wave=(bEnabled=True,WaveFrequency=(X=0.5,Y=0.6,Z=0.8),WaveAmplitude=(X=0.2,Y=0,Z=0.3),WaveOffset=(X=0,Y=1,Z=0),WaveNoise=(X=0.2,Y=0.3,Z=0.5),WaveMinimum=0,WaveMaximum=1,WaveEase=SineEaseInOut)
    WaveCurve="(EditorCurveData=(Keys=((InterpMode=RCIM_Linear,TangentMode=RCTM_Auto,TangentWeightMode=RCTWM_WeightedNone,Time=0.0176,Value=0.3729,ArriveTangent=0,ArriveTangentWeight=0,LeaveTangent=0,LeaveTangentWeight=0),(InterpMode=RCIM_Linear,TangentMode=RCTM_Auto,Tan ...(truncated)"
    bDrawDebug=True
   )
   > &RigUnit_ChainHarmonicsPerItem(
    ChainRoot=(Type=Bone,Name="deform_tentacle_l_b_1")
    DrawWorldOffset=(Rotation=(X=0,Y=0,Z=0,W=1),Translation=(X=0,Y=0,Z=0),Scale3D=(X=1,Y=1,Z=1))
    Pendulum=(bEnabled=False,PendulumStiffness=2,PendulumGravity=(X=0,Y=0,Z=0),PendulumBlend=0.75,PendulumDrag=0.98,PendulumMinimum=0,PendulumMaximum=0.1,PendulumEase=Linear,UnwindAxis=(X=0,Y=1,Z=0),UnwindMinimum=0.2,UnwindMaximum=0.05)
    Reach=(bEnabled=False,ReachTarget=(X=0,Y=0,Z=0),ReachAxis=(X=1,Y=0,Z=0),ReachMinimum=0,ReachMaximum=0,ReachEase=Linear)
    Speed=(X=0.8,Y=1.3,Z=1.5)
    Wave=(bEnabled=True,WaveFrequency=(X=0.5,Y=0.6,Z=0.8),WaveAmplitude=(X=0.1,Y=0.5,Z=0.5),WaveOffset=(X=0,Y=1,Z=0),WaveNoise=(X=0.2,Y=0.3,Z=0.5),WaveMinimum=0,WaveMaximum=1,WaveEase=SineEaseInOut)
    WaveCurve="(EditorCurveData=(Keys=((InterpMode=RCIM_Linear,TangentMode=RCTM_Auto,TangentWeightMode=RCTWM_WeightedNone,Time=0.0176,Value=0.3729,ArriveTangent=0,ArriveTangentWeight=0,LeaveTangent=0,LeaveTangentWeight=0),(InterpMode=RCIM_Linear,TangentMode=RCTM_Auto,Tan ...(truncated)"
    bDrawDebug=True
   )
   > &RigUnit_ChainHarmonicsPerItem(
    ChainRoot=(Type=Bone,Name="deform_tentacle_l_CH")
    DrawWorldOffset=(Rotation=(X=0,Y=0,Z=0,W=1),Translation=(X=0,Y=0,Z=0),Scale3D=(X=1,Y=1,Z=1))
    Pendulum=(bEnabled=False,PendulumStiffness=2,PendulumGravity=(X=0,Y=0,Z=0),PendulumBlend=0.75,PendulumDrag=0.98,PendulumMinimum=0,PendulumMaximum=0.1,PendulumEase=Linear,UnwindAxis=(X=0,Y=1,Z=0),UnwindMinimum=0.2,UnwindMaximum=0.05)
    Reach=(bEnabled=False,ReachTarget=(X=0,Y=0,Z=0),ReachAxis=(X=1,Y=0,Z=0),ReachMinimum=0,ReachMaximum=0,ReachEase=Linear)
    Speed=(X=0.7,Y=0.9,Z=1.4)
    Wave=(bEnabled=True,WaveFrequency=(X=0.3,Y=0.5,Z=0.6),WaveAmplitude=(X=0.3,Y=0,Z=0.2),WaveOffset=(X=1,Y=1,Z=3),WaveNoise=(X=0.1,Y=0.2,Z=0.2),WaveMinimum=0,WaveMaximum=1,WaveEase=SineEaseInOut)
    WaveCurve="(EditorCurveData=(Keys=((InterpMode=RCIM_Linear,TangentMode=RCTM_Auto,TangentWeightMode=RCTWM_WeightedNone,Time=0.003,Value=0.0215,ArriveTangent=0,ArriveTangentWeight=0,LeaveTangent=0,LeaveTangentWeight=0),(InterpMode=RCIM_Linear,TangentMode=RCTM_Auto,Tang ...(truncated)"
    bDrawDebug=True
   )
   > &RigUnit_ChainHarmonicsPerItem(
    ChainRoot=(Type=Bone,Name="deform_tentacle_l_2")
    DrawWorldOffset=(Rotation=(X=0,Y=0,Z=0,W=1),Translation=(X=0,Y=0,Z=0),Scale3D=(X=1,Y=1,Z=1))
    Pendulum=(bEnabled=False,PendulumStiffness=2,PendulumGravity=(X=0,Y=0,Z=0),PendulumBlend=0.75,PendulumDrag=0.98,PendulumMinimum=0,PendulumMaximum=0.1,PendulumEase=Linear,UnwindAxis=(X=0,Y=1,Z=0),UnwindMinimum=0.2,UnwindMaximum=0.05)
    Reach=(bEnabled=False,ReachTarget=(X=0,Y=0,Z=0),ReachAxis=(X=1,Y=0,Z=0),ReachMinimum=0,ReachMaximum=0,ReachEase=Linear)
    Speed=(X=0.6,Y=1.3,Z=1.3)
    Wave=(bEnabled=True,WaveFrequency=(X=0.5,Y=0.6,Z=0.8),WaveAmplitude=(X=0.2,Y=0,Z=0.2),WaveOffset=(X=0,Y=1,Z=0),WaveNoise=(X=0.2,Y=0.3,Z=0.5),WaveMinimum=0,WaveMaximum=1,WaveEase=SineEaseInOut)
    WaveCurve="(EditorCurveData=(Keys=((InterpMode=RCIM_Linear,TangentMode=RCTM_Auto,TangentWeightMode=RCTWM_WeightedNone,Time=0,Value=0,ArriveTangent=0,ArriveTangentWeight=0,LeaveTangent=0,LeaveTangentWeight=0),(InterpMode=RCIM_Linear,TangentMode=RCTM_Auto,TangentWeight ...(truncated)"
    bDrawDebug=True
   )
   > &RigUnit_ChainHarmonicsPerItem(
    ChainRoot=(Type=Bone,Name="deform_tentacle_l_fr_CH")
    DrawWorldOffset=(Rotation=(X=0,Y=0,Z=0,W=1),Translation=(X=0,Y=0,Z=0),Scale3D=(X=1,Y=1,Z=1))
    Pendulum=(bEnabled=False,PendulumStiffness=2,PendulumGravity=(X=0,Y=0,Z=0),PendulumBlend=0.75,PendulumDrag=0.98,PendulumMinimum=0,PendulumMaximum=0.1,PendulumEase=Linear,UnwindAxis=(X=0,Y=1,Z=0),UnwindMinimum=0.2,UnwindMaximum=0.05)
    Reach=(bEnabled=False,ReachTarget=(X=0,Y=0,Z=0),ReachAxis=(X=1,Y=0,Z=0),ReachMinimum=0,ReachMaximum=0,ReachEase=Linear)
    Speed=(X=0.6,Y=1.6,Z=1)
    Wave=(bEnabled=True,WaveFrequency=(X=0.2,Y=0.4,Z=0.2),WaveAmplitude=(X=0,Y=0,Z=0.2),WaveOffset=(X=0.5,Y=0.5,Z=1.5),WaveNoise=(X=0,Y=0,Z=0.1),WaveMinimum=0,WaveMaximum=1,WaveEase=SineEaseInOut)
    WaveCurve="(EditorCurveData=(Keys=((InterpMode=RCIM_Linear,TangentMode=RCTM_Auto,TangentWeightMode=RCTWM_WeightedNone,Time=0,Value=0,ArriveTangent=0,ArriveTangentWeight=0,LeaveTangent=0,LeaveTangentWeight=0),(InterpMode=RCIM_Linear,TangentMode=RCTM_Auto,TangentWeight ...(truncated)"
    bDrawDebug=True
   )
   > &RigUnit_ChainHarmonicsPerItem(
    ChainRoot=(Type=Bone,Name="deform_tentacle_l_fr_3")
    DrawWorldOffset=(Rotation=(X=0,Y=0,Z=0,W=1),Translation=(X=0,Y=0,Z=0),Scale3D=(X=1,Y=1,Z=1))
    Pendulum=(bEnabled=False,PendulumStiffness=2,PendulumGravity=(X=0,Y=0,Z=0),PendulumBlend=0.75,PendulumDrag=0.98,PendulumMinimum=0,PendulumMaximum=0.1,PendulumEase=Linear,UnwindAxis=(X=0,Y=1,Z=0),UnwindMinimum=0.2,UnwindMaximum=0.05)
    Reach=(bEnabled=False,ReachTarget=(X=0,Y=0,Z=0),ReachAxis=(X=1,Y=0,Z=0),ReachMinimum=0,ReachMaximum=0,ReachEase=Linear)
    Speed=(X=0.6,Y=1.6,Z=1)
    Wave=(bEnabled=True,WaveFrequency=(X=0.2,Y=0.4,Z=0.2),WaveAmplitude=(X=0,Y=0,Z=0.9),WaveOffset=(X=0.5,Y=0.5,Z=1.5),WaveNoise=(X=0,Y=0,Z=0.3),WaveMinimum=0,WaveMaximum=1,WaveEase=SineEaseInOut)
    WaveCurve="(EditorCurveData=(Keys=((InterpMode=RCIM_Linear,TangentMode=RCTM_Auto,TangentWeightMode=RCTWM_WeightedNone,Time=0,Value=0,ArriveTangent=0,ArriveTangentWeight=0,LeaveTangent=0,LeaveTangentWeight=0),(InterpMode=RCIM_Linear,TangentMode=RCTM_Auto,TangentWeight ...(truncated)"
    bDrawDebug=True
   )
   > &RigUnit_ChainHarmonicsPerItem(
    ChainRoot=(Type=Bone,Name="deform_tentacle_fr_CH")
    DrawWorldOffset=(Rotation=(X=0,Y=0,Z=0,W=1),Translation=(X=0,Y=0,Z=0),Scale3D=(X=1,Y=1,Z=1))
    Pendulum=(bEnabled=False,PendulumStiffness=2,PendulumGravity=(X=0,Y=0,Z=0),PendulumBlend=0.75,PendulumDrag=0.98,PendulumMinimum=0,PendulumMaximum=0.1,PendulumEase=Linear,UnwindAxis=(X=0,Y=1,Z=0),UnwindMinimum=0.2,UnwindMaximum=0.05)
    Reach=(bEnabled=False,ReachTarget=(X=0,Y=0,Z=0),ReachAxis=(X=1,Y=0,Z=0),ReachMinimum=0,ReachMaximum=0,ReachEase=Linear)
    Speed=(X=0.6,Y=1.5,Z=1.6)
    Wave=(bEnabled=True,WaveFrequency=(X=0.2,Y=0.4,Z=0.2),WaveAmplitude=(X=0,Y=0.3,Z=0.4),WaveOffset=(X=0,Y=0.5,Z=1.5),WaveNoise=(X=0,Y=0.1,Z=0.1),WaveMinimum=0,WaveMaximum=1,WaveEase=SineEaseInOut)
    WaveCurve="(EditorCurveData=(Keys=((InterpMode=RCIM_Linear,TangentMode=RCTM_Auto,TangentWeightMode=RCTWM_WeightedNone,Time=0,Value=0,ArriveTangent=0,ArriveTangentWeight=0,LeaveTangent=0,LeaveTangentWeight=0),(InterpMode=RCIM_Linear,TangentMode=RCTM_Auto,TangentWeight ...(truncated)"
    bDrawDebug=True
   )
   > &RigUnit_ChainHarmonicsPerItem(
    ChainRoot=(Type=Bone,Name="deform_tentacle_fr_2")
    DrawWorldOffset=(Rotation=(X=0,Y=0,Z=0,W=1),Translation=(X=0,Y=0,Z=0),Scale3D=(X=1,Y=1,Z=1))
    Pendulum=(bEnabled=False,PendulumStiffness=2,PendulumGravity=(X=0,Y=0,Z=0),PendulumBlend=0.75,PendulumDrag=0.98,PendulumMinimum=0,PendulumMaximum=0.1,PendulumEase=Linear,UnwindAxis=(X=0,Y=1,Z=0),UnwindMinimum=0.2,UnwindMaximum=0.05)
    Reach=(bEnabled=False,ReachTarget=(X=0,Y=0,Z=0),ReachAxis=(X=1,Y=0,Z=0),ReachMinimum=0,ReachMaximum=0,ReachEase=Linear)
    Speed=(X=0.9,Y=1.3,Z=1.3)
    Wave=(bEnabled=True,WaveFrequency=(X=0.1,Y=0.1,Z=0.1),WaveAmplitude=(X=0.1,Y=0.7,Z=0.7),WaveOffset=(X=-0.7,Y=0.8,Z=0.6),WaveNoise=(X=0.1,Y=0.5,Z=0.2),WaveMinimum=0,WaveMaximum=1,WaveEase=SineEaseInOut)
    WaveCurve="(EditorCurveData=(Keys=((InterpMode=RCIM_Linear,TangentMode=RCTM_Auto,TangentWeightMode=RCTWM_WeightedNone,Time=0,Value=0,ArriveTangent=0,ArriveTangentWeight=0,LeaveTangent=0,LeaveTangentWeight=0),(InterpMode=RCIM_Linear,TangentMode=RCTM_Auto,TangentWeight ...(truncated)"
    bDrawDebug=True
   )
   > &RigUnit_ChainHarmonicsPerItem(
    ChainRoot=(Type=Bone,Name="deform_tentacle_r_fr_CH")
    DrawWorldOffset=(Rotation=(X=0,Y=0,Z=0,W=1),Translation=(X=0,Y=0,Z=0),Scale3D=(X=1,Y=1,Z=1))
    Pendulum=(bEnabled=False,PendulumStiffness=2,PendulumGravity=(X=0,Y=0,Z=0),PendulumBlend=0.75,PendulumDrag=0.98,PendulumMinimum=0,PendulumMaximum=0.1,PendulumEase=Linear,UnwindAxis=(X=0,Y=1,Z=0),UnwindMinimum=0.2,UnwindMaximum=0.05)
    Reach=(bEnabled=False,ReachTarget=(X=0,Y=0,Z=0),ReachAxis=(X=1,Y=0,Z=0),ReachMinimum=0,ReachMaximum=0,ReachEase=Linear)
    Speed=(X=1.3,Y=1.2,Z=1.4)
    Wave=(bEnabled=True,WaveFrequency=(X=0.15,Y=0.2,Z=0.1),WaveAmplitude=(X=0.1,Y=0,Z=0.5),WaveOffset=(X=-0.3,Y=-0.5,Z=0.2),WaveNoise=(X=0.05,Y=0.02,Z=0.1),WaveMinimum=0.5,WaveMaximum=1,WaveEase=SineEaseInOut)
    WaveCurve="(EditorCurveData=(Keys=((InterpMode=RCIM_Linear,TangentMode=RCTM_Auto,TangentWeightMode=RCTWM_WeightedNone,Time=0,Value=0,ArriveTangent=0,ArriveTangentWeight=0,LeaveTangent=0,LeaveTangentWeight=0),(InterpMode=RCIM_Cubic,TangentMode=RCTM_Auto,TangentWeightM ...(truncated)"
    bDrawDebug=True
   )
   > &RigUnit_ChainHarmonicsPerItem(
    ChainRoot=(Type=Bone,Name="deform_tentacle_r_fr_2")
    DrawWorldOffset=(Rotation=(X=0,Y=0,Z=0,W=1),Translation=(X=0,Y=0,Z=0),Scale3D=(X=1,Y=1,Z=1))
    Pendulum=(bEnabled=False,PendulumStiffness=2,PendulumGravity=(X=0,Y=0,Z=0),PendulumBlend=0.75,PendulumDrag=0.98,PendulumMinimum=0,PendulumMaximum=0.1,PendulumEase=Linear,UnwindAxis=(X=0,Y=1,Z=0),UnwindMinimum=0.2,UnwindMaximum=0.05)
    Reach=(bEnabled=False,ReachTarget=(X=0,Y=0,Z=0),ReachAxis=(X=1,Y=0,Z=0),ReachMinimum=0,ReachMaximum=0,ReachEase=Linear)
    Speed=(X=1,Y=1.2,Z=1.2)
    Wave=(bEnabled=True,WaveFrequency=(X=0.15,Y=0.2,Z=0.1),WaveAmplitude=(X=0.2,Y=0,Z=0.5),WaveOffset=(X=-0.7,Y=-0.5,Z=0.2),WaveNoise=(X=0.1,Y=0.2,Z=0.3),WaveMinimum=0.5,WaveMaximum=1,WaveEase=SineEaseInOut)
    WaveCurve="(EditorCurveData=(Keys=((InterpMode=RCIM_Linear,TangentMode=RCTM_Auto,TangentWeightMode=RCTWM_WeightedNone,Time=0,Value=0,ArriveTangent=0,ArriveTangentWeight=0,LeaveTangent=0,LeaveTangentWeight=0),(InterpMode=RCIM_Cubic,TangentMode=RCTM_Auto,TangentWeightM ...(truncated)"
    bDrawDebug=True
   )
   > &RigUnit_ChainHarmonicsPerItem(
    ChainRoot=(Type=Bone,Name="deform_tentacle_r_CH")
    DrawWorldOffset=(Rotation=(X=0,Y=0,Z=0,W=1),Translation=(X=0,Y=0,Z=0),Scale3D=(X=1,Y=1,Z=1))
    Pendulum=(bEnabled=False,PendulumStiffness=2,PendulumGravity=(X=0,Y=0,Z=0),PendulumBlend=0.75,PendulumDrag=0.98,PendulumMinimum=0,PendulumMaximum=0.1,PendulumEase=Linear,UnwindAxis=(X=0,Y=1,Z=0),UnwindMinimum=0.2,UnwindMaximum=0.05)
    Reach=(bEnabled=False,ReachTarget=(X=0,Y=0,Z=0),ReachAxis=(X=1,Y=0,Z=0),ReachMinimum=0,ReachMaximum=0,ReachEase=Linear)
    Speed=(X=0.6,Y=0.6,Z=0.8)
    Wave=(bEnabled=True,WaveFrequency=(X=0.3,Y=0.6,Z=0.1),WaveAmplitude=(X=0,Y=0,Z=0.1),WaveOffset=(X=0.5,Y=0,Z=0),WaveNoise=(X=0,Y=0,Z=0.1),WaveMinimum=0.5,WaveMaximum=1,WaveEase=SineEaseInOut)
    WaveCurve="(EditorCurveData=(Keys=((InterpMode=RCIM_Linear,TangentMode=RCTM_Auto,TangentWeightMode=RCTWM_WeightedNone,Time=0,Value=0,ArriveTangent=0,ArriveTangentWeight=0,LeaveTangent=0,LeaveTangentWeight=0),(InterpMode=RCIM_Linear,TangentMode=RCTM_Auto,TangentWeight ...(truncated)"
    bDrawDebug=True
   )
   > &RigUnit_ChainHarmonicsPerItem(
    ChainRoot=(Type=Bone,Name="deform_tentacle_r_3")
    DrawWorldOffset=(Rotation=(X=0,Y=0,Z=0,W=1),Translation=(X=0,Y=0,Z=0),Scale3D=(X=1,Y=1,Z=1))
    Pendulum=(bEnabled=False,PendulumStiffness=2,PendulumGravity=(X=0,Y=0,Z=0),PendulumBlend=0.75,PendulumDrag=0.98,PendulumMinimum=0,PendulumMaximum=0.1,PendulumEase=Linear,UnwindAxis=(X=0,Y=1,Z=0),UnwindMinimum=0.2,UnwindMaximum=0.05)
    Reach=(bEnabled=False,ReachTarget=(X=0,Y=0,Z=0),ReachAxis=(X=1,Y=0,Z=0),ReachMinimum=0,ReachMaximum=0,ReachEase=Linear)
    Speed=(X=1.2,Y=1.2,Z=1.5)
    Wave=(bEnabled=True,WaveFrequency=(X=0.3,Y=0.6,Z=0.1),WaveAmplitude=(X=0.1,Y=0.02,Z=0.2),WaveOffset=(X=0.5,Y=0,Z=-0.7),WaveNoise=(X=0,Y=0,Z=0.2),WaveMinimum=0.5,WaveMaximum=1,WaveEase=SineEaseInOut)
    WaveCurve="(EditorCurveData=(Keys=((InterpMode=RCIM_Linear,TangentMode=RCTM_Auto,TangentWeightMode=RCTWM_WeightedNone,Time=0,Value=0,ArriveTangent=0,ArriveTangentWeight=0,LeaveTangent=0,LeaveTangentWeight=0),(InterpMode=RCIM_Linear,TangentMode=RCTM_Auto,TangentWeight ...(truncated)"
    bDrawDebug=True
   )
   > &RigUnit_ChainHarmonicsPerItem(
    ChainRoot=(Type=Bone,Name="deform_tentacle_r_b_1")
    DrawWorldOffset=(Rotation=(X=0,Y=0,Z=0,W=1),Translation=(X=0,Y=0,Z=0),Scale3D=(X=1,Y=1,Z=1))
    Pendulum=(bEnabled=False,PendulumStiffness=2,PendulumGravity=(X=0,Y=0,Z=0),PendulumBlend=0.75,PendulumDrag=0.98,PendulumMinimum=0,PendulumMaximum=0.1,PendulumEase=Linear,UnwindAxis=(X=0,Y=1,Z=0),UnwindMinimum=0.2,UnwindMaximum=0.05)
    Reach=(bEnabled=False,ReachTarget=(X=0,Y=0,Z=0),ReachAxis=(X=1,Y=0,Z=0),ReachMinimum=0,ReachMaximum=0,ReachEase=Linear)
    Speed=(X=1.8,Y=1.6,Z=1.2)
    Wave=(bEnabled=True,WaveFrequency=(X=0.2,Y=0.4,Z=0.2),WaveAmplitude=(X=0.2,Y=0,Z=1.5),WaveOffset=(X=0,Y=1,Z=1),WaveNoise=(X=0.05,Y=0,Z=0.05),WaveMinimum=0,WaveMaximum=1,WaveEase=SineEaseInOut)
    WaveCurve="(EditorCurveData=(Keys=((InterpMode=RCIM_Linear,TangentMode=RCTM_Auto,TangentWeightMode=RCTWM_WeightedNone,Time=0,Value=0,ArriveTangent=0,ArriveTangentWeight=0,LeaveTangent=0,LeaveTangentWeight=0),(InterpMode=RCIM_Linear,TangentMode=RCTM_Auto,TangentWeight ...(truncated)"
    bDrawDebug=True
   )
   &RigUnit_ChainHarmonicsPerItem(
    ChainRoot=(Type=Bone,Name="deform_tentacle_r_2")
    DrawWorldOffset=(Rotation=(X=0,Y=0,Z=0,W=1),Translation=(X=0,Y=0,Z=0),Scale3D=(X=1,Y=1,Z=1))
    ExecutePin=()
    Pendulum=(bEnabled=False,PendulumStiffness=2,PendulumGravity=(X=0,Y=0,Z=0),PendulumBlend=0.75,PendulumDrag=0.98,PendulumMinimum=0,PendulumMaximum=0.1,PendulumEase=Linear,UnwindAxis=(X=0,Y=1,Z=0),UnwindMinimum=0.2,UnwindMaximum=0.05)
    Reach=(bEnabled=False,ReachTarget=(X=0,Y=0,Z=0),ReachAxis=(X=1,Y=0,Z=0),ReachMinimum=0,ReachMaximum=0,ReachEase=Linear)
    Speed=(X=1.2,Y=1.2,Z=1.3)
    Wave=(bEnabled=True,WaveFrequency=(X=0.3,Y=0.6,Z=0.1),WaveAmplitude=(X=0.1,Y=0.02,Z=0.1),WaveOffset=(X=0.5,Y=0,Z=-0.7),WaveNoise=(X=0,Y=0,Z=0.2),WaveMinimum=0.5,WaveMaximum=1,WaveEase=SineEaseInOut)
    WaveCurve="(EditorCurveData=(Keys=((InterpMode=RCIM_Linear,TangentMode=RCTM_Auto,TangentWeightMode=RCTWM_WeightedNone,Time=0,Value=0,ArriveTangent=0,ArriveTangentWeight=0,LeaveTangent=0,LeaveTangentWeight=0),(InterpMode=RCIM_Linear,TangentMode=RCTM_Auto,TangentWeight ...(truncated)"
    bDrawDebug=True
   )
  ]}
]
hierarchy [
 root {loc=<0 0 0>, rot=<0 0 0 1>}
 pack {parent=root, loc=<0 -0.7316 0>, rot=<0 0 0 1>}
 deform_head_base {parent=pack, loc=<1.4054 -8.5066 7.6236>, rot=<0.5 0.5 0.5 -0.5>}
 deform_head {parent=deform_head_base, loc=<0 0 0>, rot=<-0.7069 0.0167 -0.0336 -0.7063>}
 deform_eyelid_upper {parent=deform_head, loc=<4.2192 -1.9875 0>, rot=<0.505 -0.4843 0.5258 -0.4836>}
 deform_eyelid_lower {parent=deform_head, loc=<5.1683 -2.0072 0>, rot=<0.505 -0.4843 0.5258 -0.4836>}
 deform_eyelid_eye {parent=deform_head, loc=<4.6124 -0.886 -0.0953>, rot=<0.505 -0.4843 0.5258 -0.4836>}
 deform_tentacle_l_b_base {parent=pack, loc=<-2.9619 -5.7821 4.8104>, rot=<0.0954 0.0102 0.9897 -0.1063>}
 deform_tentacle_l_b_CH {parent=deform_tentacle_l_b_base, loc=<0 0 0>, rot=<0 0 0 1>}
 deform_tentacle_l_b_1 {parent=deform_tentacle_l_b_CH, loc=<0 0 0>, rot=<0 0 0 1>}
 deform_tentacle_l_b_2 {parent=deform_tentacle_l_b_1, loc=<4.9531 -0.5756 0.0213>, rot=<0.0102 -0.0954 0.1063 -0.9897>}
 deform_tentacle_l_b_3 {parent=deform_tentacle_l_b_2, loc=<4.6822 -1.7259 -3.1474>, rot=<-0.0012 0.0598 -0.0194 -0.998>}
 deform_tentacle_l_base {parent=pack, loc=<-3.4571 -8.9444 5.3196>, rot=<0 0 0.9943 -0.1068>}
 deform_tentacle_l_CH {parent=deform_tentacle_l_base, loc=<0 0 0>, rot=<0 0 0 1>}
 deform_tentacle_l_1 {parent=deform_tentacle_l_CH, loc=<0 0 0>, rot=<0 0 0 1>}
 deform_tentacle_l_2 {parent=deform_tentacle_l_1, loc=<4.9225 -0.5723 -0.8111>, rot=<0.0753 -0.251 -0.2773 -0.9244>}
 deform_tentacle_l_3 {parent=deform_tentacle_l_2, loc=<7.1002 0.2226 -0.4375>, rot=<-0.0953 0.2576 0.3759 -0.885>}
 deform_tentacle_l_fr_base {parent=pack, loc=<-1.4064 -11.3762 5.5218>, rot=<0 0 0.8397 -0.5431>}
 deform_tentacle_l_fr_CH {parent=deform_tentacle_l_fr_base, loc=<0 0 0>, rot=<0 0 0 1>}
 deform_tentacle_l_fr_1 {parent=deform_tentacle_l_fr_CH, loc=<0 0 0>, rot=<0 0 0 1>}
 deform_tentacle_l_fr_2 {parent=deform_tentacle_l_fr_1, loc=<5.9895 -0.0607 -0.352>, rot=<-0.0126 -0.2726 0.0694 -0.9595>}
 deform_tentacle_l_fr_3 {parent=deform_tentacle_l_fr_2, loc=<3.917 -0.7826 -3.2088>, rot=<0.1537 -0.5165 -0.2403 -0.8074>}
 deform_tentacle_l_fr_4 {parent=deform_tentacle_l_fr_3, loc=<4.131 0 0>, rot=<0.0179 0.1149 0.1528 -0.9814>}
 deform_tentacle_fr_base {parent=pack, loc=<0.9444 -12.8278 6.4149>, rot=<0 0 0.7533 -0.6577>}
 deform_tentacle_fr_CH {parent=deform_tentacle_fr_base, loc=<0 0 0>, rot=<0 0 0 1>}
 deform_tentacle_fr_1 {parent=deform_tentacle_fr_CH, loc=<0 0 0>, rot=<0 0 0 1>}
 deform_tentacle_fr_2 {parent=deform_tentacle_fr_1, loc=<5.1772 0.0352 -0.4518>, rot=<0 -0.0736 0 -0.9973>}
 deform_tentacle_fr_3 {parent=deform_tentacle_fr_2, loc=<1.1429 0.0602 -2.6409>, rot=<-0.1298 -0.6951 0.0235 -0.7067>}
 deform_tentacle_fr_4 {parent=deform_tentacle_fr_3, loc=<3.5938 0 0>, rot=<0.0023 0.2777 -0.0078 -0.9606>}
 deform_tentacle_r_fr_base {parent=pack, loc=<4.679 -11.8598 6.4402>, rot=<-0.0002 -0.0007 0.3381 -0.9411>}
 deform_tentacle_r_fr_CH {parent=deform_tentacle_r_fr_base, loc=<0 0 0>, rot=<0 0 0 1>}
 deform_tentacle_r_fr_1 {parent=deform_tentacle_r_fr_CH, loc=<0 0 0>, rot=<0 -0.1714 0 -0.9852>}
 deform_tentacle_r_fr_2 {parent=deform_tentacle_r_fr_1, loc=<6.6286 -0.8934 -0.6957>, rot=<-0.188 -0.2747 0.4847 -0.8089>}
 deform_tentacle_r_fr_3 {parent=deform_tentacle_r_fr_2, loc=<4.2794 0.7812 -1.5545>, rot=<0.0927 -0.2607 -0.3219 -0.9055>}
 deform_tentacle_r_fr_4 {parent=deform_tentacle_r_fr_3, loc=<3.2537 0 0>, rot=<-0.0089 -0.0239 0.3499 -0.9364>}
 deform_tentacle_r_base {parent=pack, loc=<5.43 -6.6732 5.2002>, rot=<0.0316 -0.1675 -0.1826 -0.9683>}
 deform_tentacle_r_CH {parent=deform_tentacle_r_base, loc=<0 0 0>, rot=<0 0 0 1>}
 deform_tentacle_r_1 {parent=deform_tentacle_r_CH, loc=<0 0 0>, rot=<0 0 0 1>}
 deform_tentacle_r_2 {parent=deform_tentacle_r_1, loc=<1.9887 0 0>, rot=<-0.0527 -0.1082 0.1085 -0.9868>}
 deform_tentacle_r_3 {parent=deform_tentacle_r_2, loc=<4.0535 -2.5495 -0.1746>, rot=<-0.018 -0.0602 0.392 -0.9178>}
 deform_tentacle_r_4 {parent=deform_tentacle_r_3, loc=<6.2399 -1.6062 0.0754>, rot=<-0.0199 -0.2769 0.0688 -0.9582>}
 deform_tentacle_r_5 {parent=deform_tentacle_r_4, loc=<4.4376 1.6699 -0.7551>, rot=<0.1291 -0.0238 0.1799 -0.9749>}
 deform_tentacle_r_b_1 {parent=deform_tentacle_r_2, loc=<5.537 -0.6872 -0.1093>, rot=<0 0 0.3182 -0.948>}
 deform_tentacle_r_b_2 {parent=deform_tentacle_r_b_1, loc=<3.8027 -0.4512 -3.0611>, rot=<0.0674 -0.2938 -0.2131 -0.9294>}
 deform_tentacle_b_base {parent=pack, loc=<-2.2529 -0.4217 2.903>, rot=<-0.2905 -0.0176 0.955 -0.0578>}
 deform_tentacle_b_1 {parent=deform_tentacle_b_base, loc=<0 0 0>, rot=<0 0 0 1>}
 deform_tentacle_b_2 {parent=deform_tentacle_b_1, loc=<6.6868 0.4738 -1.4524>, rot=<-0.0566 -0.51 0.0113 -0.8582>}
]
