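# Animation Blueprint for the TerrierCure backpack cosmetic that animates its tentacles with idle motion, secondary physics and procedural sway.
# A random idle/idle-break player feeds a Control Rig and pack-bone tilt into a cached pose.
# That pose runs through chained AnimDynamics nodes with planar limits, then additive Transform Bone nodes on tentacle bones.
# The update event drives most of those rotations with sine waves, phase-shifted from a shared accumulated phase.
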AnimBlueprint /BRCosmetics/Accessories/FORT_Backpacks/Backpack_M_MED_TerrierCure/Meshes/M_MED_TerrierCure_Pack_AnimBP
target_skeleton=&/M_MED_TerrierCure_Pack_Skeleton
graphs [
 AnimGraph {nodes [
   n1 = Anim Dynamics - Bone: deform_tentacle_r_2(Alpha=1, ExternalForce="0,0,0") {
    angularDampingOverride=0.99
    bOverrideAngularDamping=true
    bUseGravityOverride=true
    alphaBoolBlend {blendOption=Linear}
    boundBone {boneName=deform_tentacle_r_2}
    componentAppliedLinearAccClamp=<3000 3000 3000>
    componentLinearAccScale=<0.3 0.3 0.3>
    componentLinearVelScale=<0.3 0.3 0.3>
    gravityOverride {z=-300}
    physicsBodyDefinitions [{
      boundBone {boneName=deform_tentacle_r_2}
      boxExtents=<10 10 10>
      constraintSetup {angularLimitsMax=<0 0 10>, angularLimitsMin=<0 -30 -10>, angularTarget=<0 0 0>}
      localJointOffset=<1 0 0>
     }]
    planarLimits [{drivingBone {boneName=pack}, planeTransform {location=<0 0 9>, rotation=<rot -40 0 0>}}]
    relativeSpaceBone {boneName=root}
   }
   Anim Dynamics - Bone: deform_tentacle_r_3(ComponentPose=@n1, Alpha=1, ExternalForce="0,0,0") {
    angularDampingOverride=0.99
    bOverrideAngularDamping=true
    bUseGravityOverride=true
    alphaBoolBlend {blendOption=Linear}
    boundBone {boneName=deform_tentacle_r_3}
    componentAppliedLinearAccClamp=<3000 3000 3000>
    componentLinearAccScale=<0.3 0.3 0.3>
    componentLinearVelScale=<0.3 0.3 0.3>
    gravityOverride {z=-300}
    physicsBodyDefinitions [{
      boundBone {boneName=deform_tentacle_r_3}
      boxExtents=<10 10 10>
      constraintSetup {angularLimitsMax=<0 0 10>, angularLimitsMin=<0 -30 -10>, angularTarget=<0 0 0>}
      localJointOffset=<1 0 0>
     }]
    planarLimits [{drivingBone {boneName=pack}, planeTransform {location=<0 0 10>, rotation=<rot -40 0 0>}}]
    relativeSpaceBone {boneName=root}
   }
   Input Pose()
   n2 = Anim Dynamics - Bone: deform_tentacle_l_b_2(ComponentPose=Local To Component(LocalPose="Use cached pose 'Post  CR'"()), Alpha=1, ExternalForce="0,0,0") {
    angularDampingOverride=0.99
    bOverrideAngularDamping=true
    bUseGravityOverride=true
    alphaBoolBlend {blendOption=Linear}
    boundBone {boneName=deform_tentacle_l_b_2}
    componentAppliedLinearAccClamp=<3000 3000 3000>
    componentLinearAccScale=<0.3 0.3 0.3>
    componentLinearVelScale=<0.3 0.3 0.3>
    gravityOverride {z=-500}
    physicsBodyDefinitions [{
      boundBone {boneName=deform_tentacle_l_b_2}
      boxExtents=<10 10 10>
      constraintSetup {angularLimitsMax=<0 10 10>, angularLimitsMin=<0 -60 -10>, angularTarget=<0 0 0>}
      localJointOffset=<1 0 0>
     }]
    planarLimits [{drivingBone {boneName=pack}, planeTransform {location=<0 0 10>, rotation=<rot 40 0 0>}}]
    relativeSpaceBone {boneName=root}
   }
   n3 = Anim Dynamics - Bone: deform_tentacle_l_b_3(ComponentPose=@n2, Alpha=1, ExternalForce="0,0,0") {
    angularDampingOverride=0.99
    bOverrideAngularDamping=true
    bUseGravityOverride=true
    alphaBoolBlend {blendOption=Linear}
    boundBone {boneName=deform_tentacle_l_b_3}
    componentAppliedLinearAccClamp=<3000 3000 3000>
    componentLinearAccScale=<0.3 0.3 0.3>
    componentLinearVelScale=<0.3 0.3 0.3>
    gravityOverride {z=-500}
    physicsBodyDefinitions [{
      boundBone {boneName=deform_tentacle_l_b_3}
      boxExtents=<10 10 10>
      constraintSetup {angularLimitsMax=<0 0 20>, angularLimitsMin=<0 -30 -20>, angularTarget=<0 0 0>}
      localJointOffset=<1 0 0>
     }]
    planarLimits [{drivingBone {boneName=pack}, planeTransform {location=<-1 0 10>, rotation=<rot 45 0 0>}}]
    relativeSpaceBone {boneName=root}
   }
   n4 = Anim Dynamics - Bone: deform_tentacle_l_2(ComponentPose=@n3, Alpha=1, ExternalForce="0,0,0") {
    angularDampingOverride=0.99
    bOverrideAngularDamping=true
    bUseGravityOverride=true
    alphaBoolBlend {blendOption=Linear}
    boundBone {boneName=deform_tentacle_l_2}
    componentAppliedLinearAccClamp=<3000 3000 3000>
    componentLinearAccScale=<0.3 0.3 0.3>
    componentLinearVelScale=<0.3 0.3 0.3>
    gravityOverride {z=-250}
    physicsBodyDefinitions [{
      boundBone {boneName=deform_tentacle_l_2}
      boxExtents=<10 10 10>
      constraintSetup {angularLimitsMax=<0 0 10>, angularLimitsMin=<0 -30 -10>, angularTarget=<0 0 0>}
      localJointOffset=<1 0 0>
     }]
    planarLimits [{drivingBone {boneName=pack}, planeTransform {location=<0 0 10>, rotation=<rot 40 0 0>}}]
    relativeSpaceBone {boneName=root}
   }
   n5 = Anim Dynamics - Bone: deform_tentacle_l_fr_2(ComponentPose=@n4, Alpha=0.5, ExternalForce="0,0,0") {
    angularDampingOverride=0.99
    bOverrideAngularDamping=true
    bUseGravityOverride=true
    alphaBoolBlend {blendOption=Linear}
    boundBone {boneName=deform_tentacle_l_fr_2}
    componentAppliedLinearAccClamp=<3000 3000 3000>
    componentLinearAccScale=<0.3 0.3 0.3>
    componentLinearVelScale=<0.3 0.3 0.3>
    gravityOverride {z=-500}
    physicsBodyDefinitions [{
      boundBone {boneName=deform_tentacle_l_fr_2}
      boxExtents=<10 10 10>
      constraintSetup {angularLimitsMax=<0 10 10>, angularLimitsMin=<0 -40 -10>, angularTarget=<0 0 0>}
      localJointOffset=<1 0 0>
     }]
    planarLimits [{drivingBone {boneName=pack}, planeTransform {location=<0 -15 0>, rotation=<rot 0 0 -90>}}]
    relativeSpaceBone {boneName=root}
   }
   n6 = Anim Dynamics - Bone: deform_tentacle_l_fr_3(ComponentPose=@n5, Alpha=0.5, ExternalForce="0,0,0") {
    angularDampingOverride=0.99
    bOverrideAngularDamping=true
    bUseGravityOverride=true
    alphaBoolBlend {blendOption=Linear}
    boundBone {boneName=deform_tentacle_l_fr_3}
    componentAppliedLinearAccClamp=<3000 3000 3000>
    componentLinearAccScale=<0.3 0.3 0.3>
    componentLinearVelScale=<0.3 0.3 0.3>
    gravityOverride {z=-500}
    physicsBodyDefinitions [{
      boundBone {boneName=deform_tentacle_l_fr_3}
      boxExtents=<10 10 10>
      constraintSetup {angularLimitsMax=<0 0 20>, angularLimitsMin=<0 -40 -20>, angularTarget=<0 0 0>}
      localJointOffset=<1 0 0>
     }]
    planarLimits [{drivingBone {boneName=pack}, planeTransform {location=<0 -16 0>, rotation=<rot 0 0 -90>}}]
    relativeSpaceBone {boneName=root}
   }
   n7 = Anim Dynamics - Bone: deform_tentacle_fr_2(ComponentPose=@n6, Alpha=0.5, ExternalForce="0,0,0") {
    angularDampingOverride=0.99
    bOverrideAngularDamping=true
    bUseGravityOverride=true
    alphaBoolBlend {blendOption=Linear}
    boundBone {boneName=deform_tentacle_fr_2}
    componentAppliedLinearAccClamp=<3000 3000 3000>
    componentLinearAccScale=<0.3 0.3 0.3>
    componentLinearVelScale=<0.3 0.3 0.3>
    gravityOverride {z=-500}
    physicsBodyDefinitions [{
      boundBone {boneName=deform_tentacle_fr_2}
      boxExtents=<10 10 10>
      constraintSetup {angularLimitsMax=<0 10 10>, angularLimitsMin=<0 -40 -10>, angularTarget=<0 0 0>}
      localJointOffset=<1 0 0>
     }]
    planarLimits [{drivingBone {boneName=pack}, planeTransform {location=<0 -15 0>, rotation=<rot 0 0 -90>}}]
    relativeSpaceBone {boneName=root}
   }
   n8 = Anim Dynamics - Bone: deform_tentacle_fr_3(ComponentPose=@n7, Alpha=0.5, ExternalForce="0,0,0") {
    angularDampingOverride=0.99
    bOverrideAngularDamping=true
    bUseGravityOverride=true
    alphaBoolBlend {blendOption=Linear}
    boundBone {boneName=deform_tentacle_fr_3}
    componentAppliedLinearAccClamp=<3000 3000 3000>
    componentLinearAccScale=<0.3 0.3 0.3>
    componentLinearVelScale=<0.3 0.3 0.3>
    gravityOverride {z=-500}
    physicsBodyDefinitions [{
      boundBone {boneName=deform_tentacle_fr_3}
      boxExtents=<10 10 10>
      constraintSetup {angularLimitsMax=<0 0 20>, angularLimitsMin=<0 -40 -20>, angularTarget=<0 0 0>}
      localJointOffset=<1 0 0>
     }]
    planarLimits [{drivingBone {boneName=pack}, planeTransform {location=<0 -15 0>, rotation=<rot 0 0 -90>}}]
    relativeSpaceBone {boneName=root}
   }
   n9 = Anim Dynamics - Bone: deform_tentacle_r_fr_1(ComponentPose=@n8, Alpha=0.5, ExternalForce="0,0,0") {
    angularDampingOverride=0.99
    bOverrideAngularDamping=true
    bUseGravityOverride=true
    alphaBoolBlend {blendOption=Linear}
    boundBone {boneName=deform_tentacle_r_fr_1}
    componentAppliedLinearAccClamp=<3000 3000 3000>
    componentLinearAccScale=<0.3 0.3 0.3>
    componentLinearVelScale=<0.3 0.3 0.3>
    gravityOverride {z=-500}
    physicsBodyDefinitions [{
      boundBone {boneName=deform_tentacle_r_fr_1}
      boxExtents=<10 10 10>
      constraintSetup {angularLimitsMax=<0 10 10>, angularLimitsMin=<0 -40 -10>, angularTarget=<0 0 0>}
      localJointOffset=<1 0 0>
     }]
    planarLimits [{drivingBone {boneName=pack}, planeTransform {location=<0 -12 0>, rotation=<rot 0 0 -90>}}]
    relativeSpaceBone {boneName=root}
   }
   n10 = Anim Dynamics - Bone: deform_tentacle_r_fr_2(ComponentPose=@n9, Alpha=0.5, ExternalForce="0,0,0") {
    angularDampingOverride=0.99
    bOverrideAngularDamping=true
    bUseGravityOverride=true
    alphaBoolBlend {blendOption=Linear}
    boundBone {boneName=deform_tentacle_r_fr_2}
    componentAppliedLinearAccClamp=<3000 3000 3000>
    componentLinearAccScale=<0.3 0.3 0.3>
    componentLinearVelScale=<0.3 0.3 0.3>
    gravityOverride {z=-400}
    physicsBodyDefinitions [{
      boundBone {boneName=deform_tentacle_r_fr_2}
      boxExtents=<10 10 10>
      constraintSetup {angularLimitsMax=<0 10 10>, angularLimitsMin=<0 -40 -10>, angularTarget=<0 0 0>}
      localJointOffset=<1 0 0>
     }]
    planarLimits [{drivingBone {boneName=pack}, planeTransform {location=<0 -15 0>, rotation=<rot 0 0 -90>}}]
    relativeSpaceBone {boneName=root}
   }
   n11 = Anim Dynamics - Bone: deform_tentacle_r_fr_3(ComponentPose=@n10, Alpha=0.5, ExternalForce="0,0,0") {
    angularDampingOverride=0.99
    bOverrideAngularDamping=true
    bUseGravityOverride=true
    alphaBoolBlend {blendOption=Linear}
    boundBone {boneName=deform_tentacle_r_fr_3}
    componentAppliedLinearAccClamp=<3000 3000 3000>
    componentLinearAccScale=<0.3 0.3 0.3>
    componentLinearVelScale=<0.3 0.3 0.3>
    gravityOverride {z=-400}
    physicsBodyDefinitions [{
      boundBone {boneName=deform_tentacle_r_fr_3}
      boxExtents=<10 10 10>
      constraintSetup {angularLimitsMax=<0 0 20>, angularLimitsMin=<0 -40 -20>, angularTarget=<0 0 0>}
      localJointOffset=<1 0 0>
     }]
    planarLimits [{drivingBone {boneName=pack}, planeTransform {location=<0 -15 0>, rotation=<rot 0 0 -90>}}]
    relativeSpaceBone {boneName=root}
   }
   n12 = Anim Dynamics - Bone: deform_tentacle_r_4(ComponentPose=@n11, Alpha=1, ExternalForce="0,0,0") {
    angularDampingOverride=0.99
    bOverrideAngularDamping=true
    bUseGravityOverride=true
    alphaBoolBlend {blendOption=Linear}
    boundBone {boneName=deform_tentacle_r_4}
    componentAppliedLinearAccClamp=<3000 3000 3000>
    componentLinearAccScale=<0.3 0.3 0.3>
    componentLinearVelScale=<0.3 0.3 0.3>
    gravityOverride {z=-300}
    physicsBodyDefinitions [{
      boundBone {boneName=deform_tentacle_r_4}
      boxExtents=<10 10 10>
      constraintSetup {angularLimitsMax=<0 0 10>, angularLimitsMin=<0 -30 -10>, angularTarget=<0 0 0>}
      localJointOffset=<1 0 0>
     }]
    planarLimits [{drivingBone {boneName=pack}, planeTransform {location=<0 0 10>, rotation=<rot -40 0 0>}}]
    relativeSpaceBone {boneName=root}
   }
   n13 = Anim Dynamics - Bone: deform_tentacle_r_b_1(ComponentPose=@n12, Alpha=1, ExternalForce="0,0,0") {
    angularDampingOverride=0.99
    bOverrideAngularDamping=true
    bUseGravityOverride=true
    alphaBoolBlend {blendOption=Linear}
    boundBone {boneName=deform_tentacle_r_b_1}
    componentAppliedLinearAccClamp=<3000 3000 3000>
    componentLinearAccScale=<0.3 0.3 0.3>
    componentLinearVelScale=<0.3 0.3 0.3>
    gravityOverride {z=-100}
    physicsBodyDefinitions [{
      boundBone {boneName=deform_tentacle_r_b_1}
      boxExtents=<10 10 10>
      constraintSetup {angularLimitsMax=<0 10 10>, angularLimitsMin=<0 -30 -10>, angularTarget=<0 0 0>}
      localJointOffset=<1 0 0>
     }]
    planarLimits [{drivingBone {boneName=pack}, planeTransform {location=<0 0 10>, rotation=<rot -45 0 0>}}]
    relativeSpaceBone {boneName=root}
   }
   n14 = Anim Dynamics - Bone: deform_tentacle_r_b_2(ComponentPose=@n13, Alpha=1, ExternalForce="0,0,0") {
    angularDampingOverride=0.99
    bOverrideAngularDamping=true
    bUseGravityOverride=true
    alphaBoolBlend {blendOption=Linear}
    boundBone {boneName=deform_tentacle_r_b_2}
    componentAppliedLinearAccClamp=<3000 3000 3000>
    componentLinearAccScale=<0.3 0.3 0.3>
    componentLinearVelScale=<0.3 0.3 0.3>
    gravityOverride {z=-100}
    physicsBodyDefinitions [{
      boundBone {boneName=deform_tentacle_r_b_2}
      boxExtents=<10 10 10>
      constraintSetup {angularLimitsMax=<0 10 10>, angularLimitsMin=<0 -30 -10>, angularTarget=<0 0 0>}
      localJointOffset=<1 0 0>
     }]
    planarLimits [{drivingBone {boneName=pack}, planeTransform {location=<0 0 11>, rotation=<rot -45 0 0>}}]
    relativeSpaceBone {boneName=root}
   }
   n15 = Get TentacleRotator_PS_0_24_add() -> (TentacleRotator_PS_0_24_add:double)
   n16 = "Transform (Modify) Bone - Bone: deform_tentacle_l_1"!(
    ComponentPose=@n14
    Alpha=1
    Translation="0,0,0"
    Rotation_Roll:float=Get TentacleRotator_PS_-12_12() -> (TentacleRotator_PS_-12_12:double)
    Rotation_Pitch:float=@n15
    Rotation_Yaw:float=0
    Scale="1,1,1"
   ) -> (Pose) {rotationMode=BMM_Additive, rotationSpace=BCS_BoneSpace, alphaBoolBlend {blendOption=Linear}, boneToModify {boneName=deform_tentacle_l_1}}
   n17 = "Transform (Modify) Bone - Bone: deform_tentacle_l_fr_1"!(
    ComponentPose=@n16
    Alpha=1
    Translation="0,0,0"
    Rotation_Roll:float=Get TentacleRotator_PS_0_24() -> (TentacleRotator_PS_0_24:double)
    Rotation_Pitch:float=Get TentacleRotator_0_40() -> (TentacleRotator_0_40:double)
    Rotation_Yaw:float=0
    Scale="1,1,1"
   ) -> (Pose) {rotationMode=BMM_Additive, rotationSpace=BCS_BoneSpace, alphaBoolBlend {blendOption=Linear}, boneToModify {boneName=deform_tentacle_l_fr_1}}
   n18 = "Transform (Modify) Bone - Bone: deform_tentacle_fr_1"!(
    ComponentPose=@n17
    Alpha=1
    Translation="0,0,0"
    Rotation_Roll:float=0
    Rotation_Pitch:float=Get TentacleRotator_PS_0_24()
    Rotation_Yaw:float=0
    Scale="1,1,1"
   ) -> (Pose) {rotationMode=BMM_Additive, rotationSpace=BCS_BoneSpace, alphaBoolBlend {blendOption=Linear}, boneToModify {boneName=deform_tentacle_fr_1}}
   n19 = "Transform (Modify) Bone - Bone: deform_tentacle_r_fr_1"!(ComponentPose=@n18, Alpha=1, Translation="0,0,0", Rotation_Roll:float=0, Rotation_Pitch:float=@n15, Rotation_Yaw:float=0, Scale="1,1,1") -> (Pose) {
    rotationMode=BMM_Additive
    rotationSpace=BCS_BoneSpace
    translationMode=BMM_Additive
    alphaBoolBlend {blendOption=Linear}
    boneToModify {boneName=deform_tentacle_r_fr_1}
   }
   n20 = "Transform (Modify) Bone - Bone: deform_tentacle_r_1"!(
    ComponentPose=@n19
    Alpha=1
    Translation="0,0,0"
    Rotation_Roll:float=5
    Rotation_Pitch:float=Get TentacleRotator_var() -> (TentacleRotator_var:double)
    Rotation_Yaw:float=Get TentacleRotator_PS_3_0_6() -> (TentacleRotator_PS_3_0_6:double)
    Scale="1,1,1"
   ) -> (Pose) {rotationMode=BMM_Additive, rotationSpace=BCS_BoneSpace, alphaBoolBlend {blendOption=Linear}, boneToModify {boneName=deform_tentacle_r_1}}
   n21 = "Transform (Modify) Bone - Bone: deform_tentacle_r_b_1"!(
    ComponentPose=@n20
    Alpha=1
    Translation="0.791,-0.132,-0.023"
    Rotation_Roll:float=5
    Rotation_Pitch:float=Get TentacleRotator_var_2() -> (TentacleRotator_var_2:double)
    Rotation_Yaw:float=0
    Scale="1,1,1"
   ) -> (Pose) {
    rotationMode=BMM_Additive
    rotationSpace=BCS_BoneSpace
    translationMode=BMM_Additive
    alphaBoolBlend {blendOption=Linear}
    boneToModify {boneName=deform_tentacle_r_b_1}
    rotation=<rot -10 -0.236 6.927>
    translation=<0.791 -0.132 -0.023>
   }
   Output Pose(Result=Component To Local(ComponentPose=@n21)) {layerGroup=None}
   n22 = Random Sequence Player!() -> (Pose:Pose Link Structure) {
    is=title AnimGraphNode_RandomPlayer
    entries [
     {
      chanceToPlay=0.6
      maxLoopCount=0
      maxPlayRate=1
      minLoopCount=0
      minPlayRate=1
      sequence=/BRCosmetics/Accessories/FORT_Backpacks/Backpack_M_MED_TerrierCure/Animation/TerrierCure_Idle_PACK
      blendIn {blendOption=Linear, blendTime=0, customCurve=null}
     }
     {
      chanceToPlay=0.2
      maxLoopCount=1
      maxPlayRate=1
      minLoopCount=0
      minPlayRate=1
      sequence=/BRCosmetics/Accessories/FORT_Backpacks/Backpack_M_MED_TerrierCure/Animation/TerrierCure_IdleBreak1_PACK
      blendIn {blendOption=Linear, blendTime=0, customCurve=null}
     }
     {
      chanceToPlay=0.3
      maxLoopCount=1
      maxPlayRate=1
      minLoopCount=0
      minPlayRate=1
      sequence=/BRCosmetics/Accessories/FORT_Backpacks/Backpack_M_MED_TerrierCure/Animation/TerrierCure_IdleBreak2_PACK
      blendIn {blendOption=Linear, blendTime=0, customCurve=null}
     }
    ]
   }
   n23 = Control Rig(Source=@n22) {
    alphaBoolBlend {blendOption=Linear}
    controlRigAssetReference {blueprintRigClass=&/M_MED_TerrierCure_Pack_CtrlRig.M_MED_TerrierCure_Pack_CtrlRig_C}
   }
   n24 = "Transform (Modify) Bone - Bone: pack"(ComponentPose=Local To Component(LocalPose=@n23), Translation="0,0,0", Rotation="0,0,-7") {
    rotationMode=BMM_Additive
    translationMode=BMM_Additive
    alphaBoolBlend {blendOption=Linear}
    boneToModify {boneName=pack}
    rotation {roll=-7}
   }
   "Save cached pose 'Post  CR'"(Pose=Component To Local(ComponentPose=@n24)) {cacheName="Post  CR"}
  ]}
 EventGraph {nodes [
   Event BlueprintUpdateAnimation()
   > n1 = Sequence!(execute) -> (then_0, then_1, then_2:Exec, then_3:Exec)
   > n2 = F_SinWave!(execute, self, Amplitude:double=20, WaveSpeed:double=0.1) -> (then, Sin:double)
   > n3 = Sequence!(execute) -> (then_0, then_1)
   > Set TentacleRotator_0_40(TentacleRotator_0_40:double=float + float(A=@n2, B=20)) -> (Output_Get:double)
   Set TentacleRotator_-20_20(execute=@n3.then_1, TentacleRotator_-20_20:double=@n2) -> (Output_Get:double)
   n4 = F_SinWave_Phaseshift!(execute=@n1.then_1, self, Amplitude:double=12, Wave Speed:double=0.2) -> (then, Sin:double)
   > n5 = Sequence()
   Set TentacleRotator_PS_-12_12(execute=@n5.then_1, TentacleRotator_PS_-12_12:double=@n4) -> (Output_Get:double)
   > Set TentacleRotator_var(TentacleRotator_var:double=float + float(B=2)) -> (Output_Get:double)
   > Set TentacleRotator_var_2(TentacleRotator_var_2:double=10) -> (Output_Get:double)
   n6 = Set TentacleRotator_PS_0_24(execute=@n5.then_0, TentacleRotator_PS_0_24:double=float + float(A=@n4, B=16)) -> (Output_Get:double)
   > Set TentacleRotator_PS_0_24_add(TentacleRotator_PS_0_24_add:double=float + float(A=@n6, B=6)) -> (Output_Get:double)
   n7 = F_SinWave_Phaseshift_2!(execute=@n1.then_2, self, Amplitude:double=16, Wave Speed:double=0.15) -> (then, Sin:double)
   > Set TentacleRotator_PS_2_0_32(TentacleRotator_PS_2_0_32:double=float + float(A=@n7, B=16)) -> (Output_Get:double)
   n8 = F_SinWave_Phaseshift_3!(execute=@n1.then_3, self, Amplitude:double=3, Wave Speed:double=0.12) -> (then, Sin:double)
   > Set TentacleRotator_PS_3_0_6(TentacleRotator_PS_3_0_6:double=float + float(A=@n8, B=6)) -> (Output_Get:double)
   n9 = BreakVector(InVec=GetVelocity!(self:Actor Object Reference=TryGetPawnOwner()) -> (ReturnValue:Vector))
   Set IsGoingUp(IsGoingUp:bool=AND Boolean(A="float > float"(A=@n9.Z), B=Get bIsJumping())) -> (Output_Get:bool)
   n10 = Get bIsSkyDiving()
   > Set IsGoingDown(IsGoingDown:bool=AND Boolean(A="float < float"(A=@n9.Z), B=NOT Boolean(A=@n10))) -> (Output_Get:bool)
  ]}
 F_SinWave {nodes [
   n1 = F_SinWave!() -> (then, Amplitude:double=~, WaveSpeed:double=~) {is=title K2Node_FunctionEntry}
   n2 = float * float!(A=@n1.WaveSpeed, B=GetWorldDeltaSeconds!() -> (ReturnValue:double), C:double=6.2814) -> (ReturnValue) {is=title K2Node_PromotableOperator}
   n3 = Get Phase() -> (Phase:double)
   > n4 = Set Phase(Phase:double="% (Float)"!(A:double=float + float(A=@n2, B=@n3), B:double=6.2814) -> (ReturnValue:double)) -> (Output_Get:double)
   > n5 = Set SinFloat(SinFloat:double=float * float!(A="Sin (Radians)"(A=@n4), B=@n1) -> (ReturnValue)) -> (Output_Get:double)
   > Return Node!(execute, Sin:double=@n5)
  ]}
 F_SinWave_Phaseshift {nodes [
   n1 = F_SinWave_Phaseshift!() -> (then, Amplitude:double=~, Wave Speed:double=~) {is=title K2Node_FunctionEntry}
   n2 = float * float!(A=@n1.Wave Speed, B=GetWorldDeltaSeconds(), C:double=6.2814) -> (ReturnValue)
   n3 = Get Phase()
   > n4 = Set PhaseShifted(PhaseShifted:double="% (Float)"(A=float + float!(A=@n2, B=@n3, C:double=1) -> (ReturnValue), B=6.2814)) -> (Output_Get:double)
   > n5 = Set SinFloat_Phaseshift(SinFloat_Phaseshift:double=float * float!(A="Sin (Radians)"(A=@n4), B=@n1) -> (ReturnValue)) -> (Output_Get:double)
   > Return Node(Sin=@n5)
  ]}
 F_SinWave_Phaseshift_2 {nodes [
   n1 = F_SinWave_Phaseshift_2!() -> (then, Amplitude:double=~, Wave Speed:double=~) {is=title K2Node_FunctionEntry}
   n2 = float * float!(A=@n1.Wave Speed, B=GetWorldDeltaSeconds(), C:double=6.2814) -> (ReturnValue)
   n3 = Get Phase()
   > n4 = Set PhaseShifted(PhaseShifted="% (Float)"(A=float + float(A=@n2, B=@n3, C=-0.5), B=6.2814))
   > n5 = Set SinFloat_Phaseshift(SinFloat_Phaseshift=float * float!(A="Sin (Radians)"(A=@n4), B=@n1) -> (ReturnValue))
   > Return Node(Sin=@n5)
  ]}
 F_SinWave_Phaseshift_3 {nodes [
   n1 = F_SinWave_Phaseshift_3!() -> (then, Amplitude:double=~, Wave Speed:double=~) {is=title K2Node_FunctionEntry}
   n2 = float * float!(A=@n1.Wave Speed, B=GetWorldDeltaSeconds(), C:double=6.2814) -> (ReturnValue)
   n3 = Get Phase()
   > n4 = Set PhaseShifted(PhaseShifted="% (Float)"(A=float + float(A=@n2, B=@n3, C=-0.8), B=6.2814))
   > n5 = Set SinFloat_Phaseshift(SinFloat_Phaseshift=float * float!(A="Sin (Radians)"(A=@n4), B=@n1) -> (ReturnValue))
   > Return Node(Sin=@n5)
  ]}
]
variables [
 IsGoingDown {type=boolean}
 IsGoingUp {type=boolean}
 Phase {type=number}
 PhaseShifted {type=number}
 SinFloat {type=number}
 SinFloat_Phaseshift {type=number}
 TentacleRotator_-20_20 {type=number}
 TentacleRotator_0_40 {type=number}
 TentacleRotator_PS_-12_12 {type=number}
 TentacleRotator_PS_0_24 {type=number}
 TentacleRotator_PS_0_24_add {type=number}
 TentacleRotator_PS_2_0_32 {type=number}
 TentacleRotator_PS_3_0_6 {type=number}
 TentacleRotator_var {type=number}
 TentacleRotator_var_2 {type=number}
]
